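# Sound cue for the end of a drift on the Super sedan in the modded basic car set, built from Tuner sedan drift-end samples.
# A random node picks one of four equally weighted wave players, and a modulator at the root processes the result.
# Playback uses a close-detail vehicle movement attenuation, the vehicle interior sound class and a reduced volume.
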SoundCue /ModdedBasicCar/Audio/Veh_Sedan_Super_Drift_End_Cue
root=Modulator_0
nodes [
 {id=Modulator_0, children [Random_0]}
 {id=Random_0, children [WavePlayer_0,WavePlayer_1,WavePlayer_2,WavePlayer_3], properties {weights [1,1,1,1]}}
 {id=WavePlayer_0, properties {soundWaveAssetPtr=&/Veh_Sedan_Tuner_Drift_End_01}}
 {id=WavePlayer_1, properties {soundWaveAssetPtr=&/Veh_Sedan_Tuner_Drift_End_02}}
 {id=WavePlayer_2, properties {soundWaveAssetPtr=&/Veh_Sedan_Tuner_Drift_End_03}}
 {id=WavePlayer_3, properties {soundWaveAssetPtr=&/Veh_Sedan_Tuner_Drift_End_04}}
]
properties {
 attenuationSettings=/VehicleAudio/Settings/Vehicle_Movement_Close_Detail_Att
 maxDistance=12000
 soundClassObject=&Fort_Vehicles_Interior
 volumeMultiplier=0.7
}
