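# Drift-start sound cue for a super sedan vehicle, built from Tuner sedan drift samples.
# A Mixer layers one of four random start one-shots with a looping drift sample, weighting the loop higher.
# The loop runs through a Modulator and an Enveloper whose volume curve fades it in from silence after the initial hit.
# Close-detail vehicle attenuation, an interior vehicle sound class and a concurrency override govern playback.
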SoundCue /ModdedBasicCar/Audio/Veh_Sedan_Super_Drift_Start_Cue
root=Mixer_0
nodes [
 {
  id=Enveloper_0
  children [Modulator_0]
  properties {pitchCurve {editorCurveData {keys [{time=0}]}}, volumeCurve {editorCurveData {keys [<at 0 0>,<at 0.7482 0.1685>,<at 1.5 1>]}}}
 }
 {id=Mixer_0, children [Enveloper_0,Random_0], properties {inputVolume [2,1]}}
 {id=Modulator_0, children [WavePlayer_0]}
 {id=Random_0, children [WavePlayer_1, WavePlayer_5, WavePlayer_6, WavePlayer_17], properties {weights [1,1,1,1]}}
 {id=WavePlayer_0, properties {bLooping=true, soundWaveAssetPtr=&/Veh_Sedan_Tuner_Drift_Loop_01}}
 {id=WavePlayer_1, properties {soundWaveAssetPtr=&/Veh_Sedan_Tuner_Drift_Start_01}}
 {id=WavePlayer_17, properties {soundWaveAssetPtr=&/Veh_Sedan_Tuner_Drift_Start_04}}
 {id=WavePlayer_5, properties {soundWaveAssetPtr=&/Veh_Sedan_Tuner_Drift_Start_02}}
 {id=WavePlayer_6, properties {soundWaveAssetPtr=&/Veh_Sedan_Tuner_Drift_Start_03}}
]
properties {
 attenuationSettings=/VehicleAudio/Settings/Vehicle_Movement_Close_Detail_Att
 bOverrideConcurrency=true
 maxDistance=12000
 soundClassObject=&Fort_Vehicles_Interior
 volumeMultiplier=0.7
 concurrencyOverrides {maxCount=3, retriggerTime=0.5}
}
