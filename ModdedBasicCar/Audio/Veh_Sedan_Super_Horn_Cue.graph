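# Sound cue for the Super sedan's horn in the ModdedBasicCar content, routed to the exterior vehicle sound class.
# A wave player holding a Tuner sedan horn wave feeds a "Horn" wave parameter node, so the sound can be swapped at runtime.
# A continuous modulator at the root maps the PitchOffset parameter to a slight pitch shift.
# It uses shared horn attenuation but the vehicle engine concurrency setting.
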SoundCue /ModdedBasicCar/Audio/Veh_Sedan_Super_Horn_Cue
root=ModulatorContinuous_0
nodes [
 {
  id=ModulatorContinuous_0
  children [WaveParam_0]
  properties {pitchModulationParams {default=0.5, maxOutput=1.05, minOutput=0.95, parameterName=PitchOffset}}
 }
 {id=WaveParam_0, children [WavePlayer_0], properties {waveParameterName=Horn}}
 {id=WavePlayer_0, properties {soundWaveAssetPtr=&/Veh_Sedan_Tuner_Horn_01}}
]
properties {
 attenuationSettings=/VehicleAudio/Settings/Vehicle_Horn_Att
 maxDistance=16030.001
 soundClassObject=&Fort_Vehicles_Exterior
 volumeMultiplier=1.268
 concurrencySet [/VehicleAudio/Settings/Vehicle_Engine_Conc]
}
