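# Sound cue for the ModdedBasicCar super sedan's gear shift, playing tuner inline-six drift upshift blowoff sounds.
# A Random node picks one of seven equally weighted wave variations and feeds it through a Modulator at its default settings.
# The cue plays at reduced volume in the exterior vehicle sound class, using shared close-detail vehicle movement attenuation.
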SoundCue /ModdedBasicCar/Audio/Veh_Sedan_Super_Shift_Cue
root=Modulator_1
nodes [
 {id=Modulator_1, children [Random_2]}
 {
  id=Random_2
  children [WavePlayer_8, WavePlayer_9, WavePlayer_10, WavePlayer_11, WavePlayer_12, WavePlayer_13, WavePlayer_14]
  properties {weights [1,1,1,1,1,1,1]}
 }
 {id=WavePlayer_10, properties {soundWaveAssetPtr=&/Veh_Sedan_Tuner_IL6_Drift_ShiftUp_Blowoff_03}}
 {id=WavePlayer_11, properties {soundWaveAssetPtr=&/Veh_Sedan_Tuner_IL6_Drift_ShiftUp_Blowoff_04}}
 {id=WavePlayer_12, properties {soundWaveAssetPtr=&/Veh_Sedan_Tuner_IL6_Drift_ShiftUp_Blowoff_05}}
 {id=WavePlayer_13, properties {soundWaveAssetPtr=&/Veh_Sedan_Tuner_IL6_Drift_ShiftUp_Blowoff_06}}
 {id=WavePlayer_14, properties {soundWaveAssetPtr=&/Veh_Sedan_Tuner_IL6_Drift_ShiftUp_Blowoff_07}}
 {id=WavePlayer_8, properties {soundWaveAssetPtr=&/Veh_Sedan_Tuner_IL6_Drift_ShiftUp_Blowoff_01}}
 {id=WavePlayer_9, properties {soundWaveAssetPtr=&/Veh_Sedan_Tuner_IL6_Drift_ShiftUp_Blowoff_02}}
]
properties {
 attenuationSettings=/VehicleAudio/Settings/Vehicle_Movement_Close_Detail_Att
 maxDistance=12000
 soundClassObject=&Fort_Vehicles_Exterior
 volumeMultiplier=0.5
}
