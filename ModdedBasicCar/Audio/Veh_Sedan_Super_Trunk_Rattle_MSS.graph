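# Looping mono MetaSound source for a sedan trunk rattle driven by a Bass input and silenced as IsLocalPassenger reaches 1.
# A waveshaped vibration loop is gain-modulated by bass rises, found by subtracting a smoothed, deferred copy of the mapped bass from its current value.
# A value comparison triggers an ADSR whose stage events switch the interpolation time between attack, decay and release, before a final limiter.
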MetaSoundSource /ModdedBasicCar/Audio/Veh_Sedan_Super_Trunk_Rattle_MSS
inputs [
 Amount {type=Float}
 Bass {type=Float}
 BassGain {type=Float}
 In {type=Float}
 InterpAttack {type=Time}
 InterpDecay {type=Time}
 InterpRelease {type=Time}
 IsLocalPassenger {type=Float}
 RattleGain {type=Float}
 UE.Source.OnPlay {type=Trigger}
]
interfaces [UE.OutputFormat.Mono:1.0, UE.Source:1.1]
outputs ["UE.OutputFormat.Mono.Audio:0" {type=Audio}]
pages [Default {
  input_defaults {
   Amount=2
   Bass=0
   BassGain=5
   In=0
   InterpAttack=0.1
   InterpDecay=0.01
   InterpRelease=0.5
   IsLocalPassenger=0
   RattleGain=2
   UE.Source.OnPlay=null
  }
  nodes [
   InitVariable.Float()
   n1 = VariableDeferredAccessor.Float(Variable=InitVariable.Float().Variable)
   n2 = MapRange.MapRange:Float(In=^Bass)
   n3 = UE.Subtract:Float(AdditionalOperands=UE.InterpTo:Audio(Target=@n1.Value).Value, PrimaryOperand=@n2.Out Value)
   n4 = VariableAccessor.Float(Variable=VariableMutator.Float(Variable=InitVariable.Float().Variable, Value=Max.Max:Float(A=@n3.Out).Value).Variable)
   n5 = VariableAccessor.Float(Variable=@n4.Variable)
   n6 = VariableAccessor.Float(Variable=@n5.Variable)
   n7 = VariableDeferredAccessor.Time(Variable=InitVariable.Time().Variable)
   WavePlayer_0 = UE.Wave Player:Mono(Play=^UE.Source.OnPlay) {also {
     defaults {Loop="<MetaSoundBuilderResult.SUCCEEDED: 0>", Wave Asset=&/Veh_Sedan_Tuner_Trunk_Vibrate_Loop_04}
     version [1,1]
    }}
   AudioMixerAudioMixerMono2_0 = "AudioMixer.Audio Mixer (Mono, 2)"(
    Gain 0=UE.InterpTo:Audio(Target=UE.Multiply:Float(PrimaryOperand=@n6.Value) {also {defaults {AdditionalOperands=10}}}.Out, Interp Time=@n7.Value).Value
    In 0=UE.WaveShaper:Audio(OutputGain=^RattleGain, In=@WavePlayer_0.Out Mono).Out
   )
   Limiter_0 = UE.Limiter:Audio(Audio=UE.Multiply:Audio by Float(
     PrimaryOperand=@AudioMixerAudioMixerMono2_0.Out
     AdditionalOperands=UE.Subtract:Float(AdditionalOperands=^IsLocalPassenger) {also {defaults {PrimaryOperand=1}}}.Out
    ).Out)
   VariableMutator.Float(Value=@n2.Out Value, Variable=@n1.Variable)
   n8 = VariableDeferredAccessor.Time(Variable=@n7.Variable)
   n9 = VariableAccessor.Float(Variable=VariableMutator.Float(Variable=InitVariable.Float().Variable, Value=UE.InterpTo:Audio(Target=@n4.Value, Interp Time=@n8.Value).Value).Variable)
   TriggerOnValueChange_0 = UE.Trigger On Value Change:Float(Value=VariableAccessor.Float(Variable=@n9.Variable).Value)
   n10 = TriggerCompare.Clamp:Float(Compare=@TriggerOnValueChange_0.Trigger, A=@n9.Value, B=@n5.Value) {also {defaults {Type=2}}}
   ADSREnvelopeADSREnvelope_0 = ADSR Envelope.ADSR Envelope:Float(Trigger Release=@n10.FALSE, Trigger Attack=@n10.TRUE) {also {defaults {Attack Time=0.05, Sustain Level=1}}}
   TriggerRouteTriggerRouteTime3_0 = "TriggerRoute.Trigger Route (Time, 3):Time"(
    Value 0=^InterpAttack
    Value 1=^InterpDecay
    Value 2=^InterpRelease
    Set 0=@ADSREnvelopeADSREnvelope_0.On Attack Triggered
    Set 1=@ADSREnvelopeADSREnvelope_0.On Decay Triggered
    Set 2=@ADSREnvelopeADSREnvelope_0.On Release Triggered
   )
   VariableMutator.Time(Value=@TriggerRouteTriggerRouteTime3_0.Value, Variable=@n8.Variable)
  ]
  links [{dst="@UE.OutputFormat.Mono.Audio:0", src=Limiter_0.Audio}]
 }]
properties {
 assetImportData=&self:AssetImportData
 attenuationSettings=&/Veh_Sedan_Super_Trunk_Rattle_Att
 bLooping=true
 duration=10000
 maxDistance=768
 soundClassObject=&Fort_Vehicles
 volume=2
}
