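# Looping trunk rattle sound cue for the Super variant of the modded sedan.
# A quality-level node at the root feeds the same looping trunk rattle wave player into its first two slots and leaves the other two empty.
# The cue plays at reduced volume in the Fort_Vehicles sound class, using a dedicated attenuation asset to limit its audible range.
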SoundCue /ModdedBasicCar/Audio/Veh_Sedan_Super_Trunk_Rattle_Cue
root=QualityLevel_0
nodes [
 {id=QualityLevel_0, children [WavePlayer_0,WavePlayer_0,null,null]}
 {id=WavePlayer_0, properties {bLooping=true, soundWaveAssetPtr=&/Veh_Sedan_Super_Trunk_Rattle_MSS}}
]
properties {
 attenuationSettings=&/Veh_Sedan_Super_Trunk_Rattle_Att
 maxDistance=768
 soundClassObject=&Fort_Vehicles
 volumeMultiplier=0.25
}
