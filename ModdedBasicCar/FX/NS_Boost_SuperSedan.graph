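# Niagara boost effect for the SuperSedan: twin exhaust flames, glow lights and sparks, plus neon rings at the rear tires.
# Exhaust emitters alternate spawn groups between two skeletal-mesh sockets, with purple-blue sparks split across GPU and CPU emitters.
# Flames, lights and rings take System.VehicleColor, and normalized loop-age curves taper the burst.
# Rings stop in deep water or when that side's TireIntact flag is true; a fixed-purple light variant is disabled.
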NiagaraSystem /ModdedBasicCar/FX/NS_Boost_SuperSedan
dependencies {
 usedDataInterfaces [
  &NiagaraDataInterfaceArrayInt32
  &NiagaraDataInterfaceArrayFloat3
  &NiagaraDataInterfaceVectorCurve
  &NiagaraDataInterfaceVector2DCurve
 ]
 usedDynamicInputs [&SelectLinearColorFromArray]
 usedModules [
  &SystemState
  &self:SetPeeloutFXArrayIndex
  &Timeline
  &self:SystemSpawnScript.SystemScriptSource.SystemScriptGraph.NiagaraNodeAssignment_0.SetVariables_#1
 ]
}
emitters {
 ExhaustFlame {
  emitterUpdateScript [
   &EmitterState()
   &SpawnRate(
    SpawnRate=&FloatFromCurve(
     FloatCurve=DataInterface() {propertyValues {Curve {keys [
         {interpMode=RCIM_Cubic, tangentMode=RCTM_User, time=0.3, value=1}
         {arriveTangent=-2.5652, interpMode=RCIM_Cubic, leaveTangent=-2.5652, tangentMode=RCTM_User, time=1, value=0.25}
        ]}}}
     CurveIndex=$Emitter.NormalizedLoopAge
     Scale Curve=30
    )
    Use Spawn Probability=true
   )
   &SpawnRate(SpawnRate=50, Use Spawn Probability=false) {enabled=false, module=SpawnRate001}
  ]
  particleSpawnScript [
   &FN_SetSpawnGroup(Number of Spawn Groups=2)
   &InitializeParticle(
    Color=$System.VehicleColor
    Lifetime Max=0.15
    Lifetime Min=0.07
    Lifetime Mode=::NewEnumerator1
    Mesh Scale=&Lerp_Vector(
     A=<0.4 0.6 0.6>
     B=<0.6 0.75 0.75>
     Alpha=&FloatFromCurve(
      FloatCurve=DataInterface() {propertyValues {Curve {keys [
          {interpMode=RCIM_Cubic, tangentMode=RCTM_Break, time=0, value=0}
          {arriveTangent=1.393, interpMode=RCIM_Cubic, tangentMode=RCTM_Break, time=1, value=1}
         ]}}}
      CurveIndex=&RandomRangeFloat(Minimum=0, Maximum=1)
     )
    )
    Mesh Scale Mode=::NewEnumerator4
    UsePositionOffset=false
    Write Material Random=true
   )
   &SkeletalMeshLocation(
    Bone / Socket Index=&SelectIntFromArray(Direct Array Index=$Particles.SpawnGroup, Int Selection Array=DataInterface() {propertyValues {IntData [12,13]}})
    Interpolate Mesh Sampling=false
    MaskSpawnGroup=false
    Mesh Sampling Type=::NewEnumerator5
    Orientation Sampling=::NewEnumerator1
    Sampled Position Offset=&SelectVectorFromArray(Direct Array Index=$Particles.SpawnGroup, Vector Selection Array=DataInterface() {propertyValues {FloatData [<-5 0 0>, <-5 0 0>]}})
    Skeletal Mesh=$User.Vehicle_Selection_02
    Socket Sampling Mode=::NewEnumerator5
   )
   &InitialMeshOrientation(
    Mesh Orientation Mode=::NewEnumerator2
    Rotation=&Multiply_VectorByFloat(
     Vector=&Lerp_Vector(A=<0 0 0>, B=<0 0 0>, Alpha=&MakeCustomFloatFromBool(Bool=$User.BoostFX, True Float=1, False Float=0))
     Float=&SelectFloatFromArray(Direct Array Index=$Particles.SpawnGroup, Float Selection Array=DataInterface() {propertyValues {FloatData [-1,1,-1,1]}})
    )
    Use Rotation Vector=true
   )
  ]
  particleUpdateScript [
   &ParticleState()
   &ScaleColor(
    Scale Alpha=&FloatFromCurve(FloatCurve=DataInterface() {propertyValues {Curve {keys [
         {interpMode=RCIM_Cubic, leaveTangent=9.8521, tangentMode=RCTM_Break, time=0, value=0.1}
         {interpMode=RCIM_Cubic, tangentMode=RCTM_User, time=0.2, value=1}
         {interpMode=RCIM_Cubic, tangentMode=RCTM_Break, time=1, value=0}
        ]}}})
    Scale RGB=&VectorFromFloat(Value=&Multiply_Float(
      A=&RandomRangeFloat(Minimum=3, Maximum=20)
      B=&FloatFromCurve(
       FloatCurve=DataInterface() {propertyValues {Curve {keys [
           {interpMode=RCIM_Cubic, tangentMode=RCTM_User, time=0.3, value=1}
           {arriveTangentWeight=0.6424, interpMode=RCIM_Cubic, tangentMode=RCTM_User, time=1, value=0.25}
          ]}}}
       CurveIndex=$Emitter.NormalizedLoopAge
       Scale Curve=3
      )
     ))
    ScaleA=true
    ScaleRGB=true
   )
   &ScaleMeshSize(Scale Factor=&VectorFromCurve(
     VectorCurve=DataInterface() {propertyValues {
       bOptimizeLUT=false
       XCurve {keys [{arriveTangent=-0.4338, interpMode=RCIM_Cubic, leaveTangent=-0.4338, tangentMode=RCTM_User, time=0, value=0.75}, <at 1 2>]}
       YCurve {keys [{arriveTangent=0.1422, interpMode=RCIM_Cubic, leaveTangent=0.1422, tangentMode=RCTM_User, time=0, value=0.75}, <at 1 2>]}
       ZCurve {keys [
         <at -0.0011 1.0067 RCIM_Cubic>
         {arriveTangent=-0.105, interpMode=RCIM_Cubic, leaveTangent=-0.105, tangentMode=RCTM_User, time=1.0016, value=1}
        ]}
      }}
     Scale Curve=&VectorFromFloat(Value=&Lerp_Float(
       A=1.3
       B=2
       Alpha=&FloatFromCurve(
        FloatCurve=DataInterface() {propertyValues {Curve {keys [
            {interpMode=RCIM_Cubic, tangentMode=RCTM_User, time=0, value=1}
            {arriveTangentWeight=1.1078, interpMode=RCIM_Cubic, tangentMode=RCTM_User, time=1, value=0}
           ]}}}
        CurveIndex=$Emitter.NormalizedLoopAge
       )
      ))
    ))
   /Game/Effects/Niagara/Modules/Materials/MaterialPanner(
    Jitter Pan by SpawnGroups=false
    Offset:NiagaraFloat=&Add_Float(A=$Emitter.Age, B=$Particles.Initial.MaterialRandom)
    Speed=-5
   )
  ]
  rendererClasses [&NiagaraMeshRendererProperties]
  renderers [{rendererClass=&NiagaraMeshRendererProperties, rendererIndex=0}]
 }
 ExhaustLight {
  emitterUpdateScript [
   &EmitterState()
   &SpawnBurst_Instantaneous(Spawn Count=2, Use Spawn Probability=false, UseLoopCountLimit=false)
  ]
  particleSpawnScript [
   &FN_SetSpawnGroup(Number of Spawn Groups=2)
   &InitializeParticle(Color=$System.VehicleColor, Sprite Size Mode=::NewEnumerator3, Uniform Sprite Size=150, UsePositionOffset=false, Write Material Random=true)
   &SkeletalMeshLocation(
    Bone / Socket Index=&SelectIntFromArray(Direct Array Index=$Particles.SpawnGroup, Int Selection Array=DataInterface() {propertyValues {IntData [12,13]}})
    Interpolate Mesh Sampling=false
    MaskSpawnGroup=false
    Mesh Sampling Type=::NewEnumerator5
    Orientation Sampling=::NewEnumerator1
    Sampled Position Offset=&SelectVectorFromArray(Direct Array Index=$Particles.SpawnGroup, Vector Selection Array=DataInterface() {propertyValues {FloatData [<-25 0 0>, <-25 0 0>]}})
    Skeletal Mesh=$User.Vehicle_Selection_02
    Socket Sampling Mode=::NewEnumerator5
   )
  ]
  particleUpdateScript [
   &ParticleState()
   &FN_Light_Attributes(
    Color=$System.VehicleColor
    Create Color Attribute=true
    Create Exponent Attribute=true
    Create Position Attribute=true
    Create Radius Attribute=true
    Position=&ConvertPositionToVector(Input Position=$Particles.Position, Convert To Absolute Large World Vector=false)
    Radius=250
    Scale Color=&Multiply_Float(
     A=&Multiply_Float(A=5, B=&Waveform("Waveform [1]"=::NewEnumerator0, "[1] Frequency"=2, "[1] Amplitude Min/Max"=<0.95 1>))
     B=&Lerp_Float(
      A=0.25
      B=1
      Alpha=&FloatFromCurve(
       FloatCurve=DataInterface() {propertyValues {Curve {keys [
           {interpMode=RCIM_Cubic, tangentMode=RCTM_User, time=0, value=1}
           {arriveTangentWeight=1.1078, interpMode=RCIM_Cubic, tangentMode=RCTM_User, time=1, value=0}
          ]}}}
       CurveIndex=$Emitter.NormalizedLoopAge
      )
     )
    )
   )
   &ScaleColor(
    Scale Alpha=&FloatFromCurve(
     FloatCurve=DataInterface() {propertyValues {Curve {keys [
         {interpMode=RCIM_Cubic, tangentMode=RCTM_User, time=0, value=0}
         {interpMode=RCIM_Cubic, tangentMode=RCTM_User, time=0.25, value=1}
         {interpMode=RCIM_Cubic, tangentMode=RCTM_User, time=1, value=0}
        ]}}}
     Scale Curve=0.15
    )
    Scale RGB=&VectorFromFloat(Value=&Lerp_Float(
      A=0.5
      B=2
      Alpha=&FloatFromCurve(FloatCurve=DataInterface() {propertyValues {Curve {keys [
           {interpMode=RCIM_Cubic, tangentMode=RCTM_User, time=0, value=0}
           {interpMode=RCIM_Cubic, tangentMode=RCTM_User, time=0.25, value=1}
           {interpMode=RCIM_Cubic, tangentMode=RCTM_User, time=1, value=0}
          ]}}})
     ))
    ScaleA=true
    ScaleRGB=true
   )
   &DynamicMaterialParameters(
    Index 1 Param 1=150
    Module.Param0WriteEnabled1=true
    Module.Param1WriteEnabled1=true
    Module.Param2WriteEnabled1=true
    Module.Param3WriteEnabled1=true
    Write Parameter Index 0=false
    Write Parameter Index 1=true
   )
  ]
  rendererClasses [&NiagaraLightRendererProperties, &NiagaraSpriteRendererProperties]
  renderers [
   {rendererClass=&NiagaraLightRendererProperties, rendererIndex=0}
   {rendererClass=&NiagaraSpriteRendererProperties, rendererIndex=1}
  ]
 }
 ExhaustLight001 {
  bEnabled=false
  emitterUpdateScript [
   &EmitterState()
   &SpawnBurst_Instantaneous(Spawn Count=2, Use Spawn Probability=false, UseLoopCountLimit=false)
  ]
  particleSpawnScript [
   &FN_SetSpawnGroup(Number of Spawn Groups=2)
   &InitializeParticle(
    Color=<rgba 0.9346 0.1957 3 1>
    Sprite Size Mode=::NewEnumerator3
    Uniform Sprite Size=150
    UsePositionOffset=false
    Write Material Random=true
   )
   &SkeletalMeshLocation(
    Bone / Socket Index=&SelectIntFromArray(Direct Array Index=$Particles.SpawnGroup, Int Selection Array=DataInterface() {propertyValues {IntData [12,13]}})
    Interpolate Mesh Sampling=false
    MaskSpawnGroup=false
    Mesh Sampling Type=::NewEnumerator5
    Orientation Sampling=::NewEnumerator1
    Sampled Position Offset=&SelectVectorFromArray(Direct Array Index=$Particles.SpawnGroup, Vector Selection Array=DataInterface() {propertyValues {FloatData [<-25 0 0>, <-25 0 0>]}})
    Skeletal Mesh=$User.Vehicle_Selection_02
    Socket Sampling Mode=::NewEnumerator5
   )
  ]
  particleUpdateScript [
   &ParticleState()
   &FN_Light_Attributes(
    Color=<rgba 0.3214 0.003 1 1>
    Create Color Attribute=true
    Create Exponent Attribute=true
    Create Position Attribute=true
    Create Radius Attribute=true
    Position=&ConvertPositionToVector(Input Position=$Particles.Position, Convert To Absolute Large World Vector=false)
    Radius=250
    Scale Color=&Multiply_Float(
     A=&Multiply_Float(A=5, B=&Waveform("Waveform [1]"=::NewEnumerator0, "[1] Frequency"=2, "[1] Amplitude Min/Max"=<0.95 1>))
     B=&Lerp_Float(
      A=0.25
      B=1
      Alpha=&FloatFromCurve(
       FloatCurve=DataInterface() {propertyValues {Curve {keys [
           {interpMode=RCIM_Cubic, tangentMode=RCTM_User, time=0, value=1}
           {arriveTangentWeight=1.1078, interpMode=RCIM_Cubic, tangentMode=RCTM_User, time=1, value=0}
          ]}}}
       CurveIndex=$Emitter.NormalizedLoopAge
      )
     )
    )
   )
   &ScaleColor(
    Scale Alpha=&FloatFromCurve(
     FloatCurve=DataInterface() {propertyValues {Curve {keys [
         {interpMode=RCIM_Cubic, tangentMode=RCTM_User, time=0, value=0}
         {interpMode=RCIM_Cubic, tangentMode=RCTM_User, time=0.25, value=1}
         {interpMode=RCIM_Cubic, tangentMode=RCTM_User, time=1, value=0}
        ]}}}
     Scale Curve=0.15
    )
    Scale RGB=&VectorFromFloat(Value=&Lerp_Float(
      A=0.5
      B=2
      Alpha=&FloatFromCurve(FloatCurve=DataInterface() {propertyValues {Curve {keys [
           {interpMode=RCIM_Cubic, tangentMode=RCTM_User, time=0, value=0}
           {interpMode=RCIM_Cubic, tangentMode=RCTM_User, time=0.25, value=1}
           {interpMode=RCIM_Cubic, tangentMode=RCTM_User, time=1, value=0}
          ]}}})
     ))
    ScaleA=true
    ScaleRGB=true
   )
   &DynamicMaterialParameters(
    Index 1 Param 1=150
    Module.Param0WriteEnabled1=true
    Module.Param1WriteEnabled1=true
    Module.Param2WriteEnabled1=true
    Module.Param3WriteEnabled1=true
    Write Parameter Index 0=false
    Write Parameter Index 1=true
   )
  ]
  rendererClasses [&NiagaraLightRendererProperties, &NiagaraSpriteRendererProperties]
  renderers [
   {rendererClass=&NiagaraLightRendererProperties, rendererIndex=0}
   {rendererClass=&NiagaraSpriteRendererProperties, rendererIndex=1}
  ]
 }
 HotSphere {
  emitterUpdateScript [
   &EmitterState(Life Cycle Mode=::NewEnumerator1, Loop Behavior=::NewEnumerator1, Loop Duration=0.8, UseLoopDelay=false)
   &SpawnRate(
    SpawnRate=&FloatFromCurve(
     FloatCurve=DataInterface() {propertyValues {Curve {keys [
         {interpMode=RCIM_Cubic, tangentMode=RCTM_User, time=0.1, value=1}
         {arriveTangent=-2.1067, interpMode=RCIM_Cubic, leaveTangent=-2.1067, tangentMode=RCTM_User, time=1, value=0.3}
        ]}}}
     CurveIndex=$Emitter.NormalizedLoopAge
     Scale Curve=25
    )
    Use Spawn Probability=false
   )
  ]
  particleSpawnScript [
   &FN_SetSpawnGroup(Number of Spawn Groups=2)
   &InitializeParticle(
    Color=$System.VehicleColor
    Lifetime Max=0.35
    Lifetime Min=0.1
    Lifetime Mode=::NewEnumerator1
    Mesh Scale=&Multiply_VectorByFloat(Vector=<1 0.2 0.2>, Float=1.2)
    Mesh Scale Mode=::NewEnumerator4
    UsePositionOffset=false
    Write Material Random=true
   )
   &SkeletalMeshLocation(
    Bone / Socket Index=&SelectIntFromArray(Direct Array Index=$Particles.SpawnGroup, Int Selection Array=DataInterface() {propertyValues {IntData [12,13]}})
    Interpolate Mesh Sampling=false
    MaskSpawnGroup=false
    Mesh Sampling Type=::NewEnumerator5
    Orientation Sampling=::NewEnumerator1
    Sampled Position Offset=&SelectVectorFromArray(Direct Array Index=$Particles.SpawnGroup, Vector Selection Array=DataInterface() {propertyValues {FloatData [<-5 0 0>, <-5 0 0>]}})
    Skeletal Mesh=$User.Vehicle_Selection_02
    Socket Sampling Mode=::NewEnumerator5
   )
  ]
  particleUpdateScript [
   &ParticleState()
   &ScaleMeshSize(Scale Factor=&MakeVector(
     X=&Lerp_Float(
      A=1
      B=0.25
      Alpha=&FloatFromCurve(FloatCurve=DataInterface() {propertyValues {Curve {keys [{interpMode=RCIM_Cubic, tangentMode=RCTM_User, time=0, value=0}, {interpMode=RCIM_Cubic, tangentMode=RCTM_User, time=1, value=1}]}}})
     )
     Y=1
     Z=1
    ))
   &ScaleColor(
    Scale Alpha=&FloatFromCurve(
     FloatCurve=DataInterface() {propertyValues {Curve {keys [
         {arriveTangent=21.395, interpMode=RCIM_Cubic, leaveTangent=21.395, tangentMode=RCTM_User, time=0, value=0}
         {interpMode=RCIM_Cubic, tangentMode=RCTM_User, time=0.1, value=1}
         {interpMode=RCIM_Cubic, tangentMode=RCTM_User, time=1, value=0}
        ]}}}
     Scale Curve=0.1
    )
    Scale RGB=&VectorFromFloat(Value=&Multiply_Float(
      A=&RandomRangeFloat(Minimum=3, Maximum=20)
      B=&FloatFromCurve(
       FloatCurve=DataInterface() {propertyValues {Curve {keys [
           {interpMode=RCIM_Cubic, tangentMode=RCTM_User, time=0.3, value=1}
           {arriveTangentWeight=0.6424, interpMode=RCIM_Cubic, tangentMode=RCTM_User, time=1, value=0.25}
          ]}}}
       CurveIndex=$Emitter.NormalizedLoopAge
      )
     ))
    ScaleA=true
    ScaleRGB=true
   )
  ]
  rendererClasses [&NiagaraMeshRendererProperties]
  renderers [{rendererClass=&NiagaraMeshRendererProperties, rendererIndex=0}]
 }
 Jet_BurstSparksHigh {
  simTarget=GPUComputeSim
  emitterUpdateScript [
   &EmitterState()
   &SpawnRate(
    SpawnRate=&FloatFromCurve(
     FloatCurve=DataInterface() {propertyValues {Curve {keys [
         {interpMode=RCIM_Cubic, tangentMode=RCTM_User, time=0.1, value=1}
         {arriveTangent=-2.5652, interpMode=RCIM_Cubic, leaveTangent=-2.5652, tangentMode=RCTM_User, time=1, value=0.1}
        ]}}}
     CurveIndex=$Emitter.NormalizedLoopAge
     Scale Curve=250
    )
    Use Spawn Probability=false
   )
  ]
  particleSpawnScript [
   &FN_SetSpawnGroup(Number of Spawn Groups=2)
   /Niagara/Modules/Spawn/Initialization/InitializeParticle(
    Color=&RandomRangeLinearColor(Minimum=<rgba 0.5071 0 0.6406 1>, Maximum=<rgba 0.0897 0.0722 1 1>)
    Lifetime=&Lerp_Float(
     A=0.15
     B=0.9
     Alpha=&FloatFromCurve(
      FloatCurve=DataInterface() {propertyValues {Curve {keys [
          {interpMode=RCIM_Cubic, tangentMode=RCTM_User, time=0, value=0}
          {arriveTangent=0.9384, interpMode=RCIM_Cubic, leaveTangent=0.9384, tangentMode=RCTM_User, time=0.85, value=0.25}
          {arriveTangent=8.8331, interpMode=RCIM_Cubic, leaveTangent=8.8331, tangentMode=RCTM_User, time=1, value=1}
         ]}}}
      CurveIndex=&RandomRangeFloat(Minimum=0, Maximum=1)
     )
    )
    Module.Write Color=true
    Module.Write Lifetime=true
    Module.Write Mass=false
    Module.Write Material Random=true
    Module.Write Position=false
    Module.Write Scale=false
    Module.Write SpriteRotation=true
    Module.Write SpriteSize=true
    Position=$Engine.Owner.Position
    Sprite Size=&Multiply_Vector2D_ByFloat(
     Float=&Lerp_Float(
      A=1
      B=3
      Alpha=&FloatFromCurve(
       FloatCurve=DataInterface() {propertyValues {Curve {keys [
           {interpMode=RCIM_Cubic, tangentMode=RCTM_User, time=0, value=0}
           {arriveTangent=0.9384, interpMode=RCIM_Cubic, leaveTangent=0.9384, tangentMode=RCTM_User, time=0.85, value=0.25}
           {arriveTangent=8.8331, interpMode=RCIM_Cubic, leaveTangent=8.8331, tangentMode=RCTM_User, time=1, value=1}
          ]}}}
       CurveIndex=&RandomRangeFloat(Minimum=0, Maximum=1)
      )
     )
     Vector2D=&UniformRangedVector2D(Minimum=<5 7>, Maximum=<4 8>)
    )
   )
   &AddVelocityInCone(
    Cone Angle=5
    Module.Use Velocity Falloff On Cone Axis=true
    Velocity Distribution Along Cone Axis=0.6
    Velocity Falloff Away From Cone Axis=0.333
    Velocity Strength=&Lerp_Float(
     A=1200
     B=2200
     Alpha=&FloatFromCurve(
      FloatCurve=DataInterface() {propertyValues {Curve {keys [
          {interpMode=RCIM_Cubic, tangentMode=RCTM_User, time=0, value=0}
          {arriveTangent=0.9384, interpMode=RCIM_Cubic, leaveTangent=0.9384, tangentMode=RCTM_User, time=0.8, value=0.35}
          {arriveTangent=4.6474, interpMode=RCIM_Cubic, leaveTangent=4.6474, tangentMode=RCTM_User, time=1, value=1}
         ]}}}
      CurveIndex=&RandomRangeFloat(Minimum=0, Maximum=1)
     )
    )
   )
   &SkeletalMeshLocation(
    Bone / Socket Index=&SelectIntFromArray(Direct Array Index=$Particles.SpawnGroup, Int Selection Array=DataInterface() {propertyValues {IntData [12,13]}})
    Interpolate Mesh Sampling=false
    MaskSpawnGroup=false
    Mesh Sampling Type=::NewEnumerator5
    Orientation Sampling=::NewEnumerator1
    Sampled Position Offset=&SelectVectorFromArray(Direct Array Index=$Particles.SpawnGroup, Vector Selection Array=DataInterface() {propertyValues {FloatData [<-5 0 0>, <-5 0 0>]}})
    Skeletal Mesh=$User.Vehicle_Selection_02
    Socket Sampling Mode=::NewEnumerator5
   )
   /Niagara/Modules/Spawn/Location/SphereLocation(MaskSpawnGroup=false, Module.Override Randomness=false, Sphere Radius=10, Surface Only=false)
  ]
  particleUpdateScript [
   &ParticleState()
   /Niagara/Modules/Update/Forces/CurlNoiseForce(
    Noise Frequency=100
    Noise Strength=&Lerp_Float(
     A=1500
     B=6500
     Alpha=&FloatFromCurve(
      FloatCurve=DataInterface() {propertyValues {Curve {keys [
          {interpMode=RCIM_Cubic, tangentMode=RCTM_User, time=0, value=0}
          {arriveTangent=0.9384, interpMode=RCIM_Cubic, leaveTangent=0.9384, tangentMode=RCTM_User, time=0.8, value=0.35}
          {arriveTangent=4.6474, interpMode=RCIM_Cubic, leaveTangent=4.6474, tangentMode=RCTM_User, time=1, value=1}
         ]}}}
      CurveIndex=&RandomRangeFloat(Minimum=0, Maximum=1)
     )
    )
    Pan Noise=false
    Random Seed=1
   )
   &Drag(Drag=5, Use Linear Drag=true, Use Rotational Drag=false)
   &SolveForcesAndVelocity(Clamp Velocity=false, Limit Acceleration=false)
   &ScaleSpriteSize(Scale Factor=&Vector2DFromCurve(
     Vector2Curve=DataInterface() {propertyValues {
       XCurve {keys [<at -0.0017 0.4556 RCIM_Cubic>, {arriveTangent=-0.9208, interpMode=RCIM_Cubic, tangentMode=RCTM_Break, time=0.1932, value=0.1104}]}
       YCurve {keys [<at 0 1 RCIM_Cubic>,<at 0.6807 0.0224 RCIM_Cubic>]}
      }}
     ScaleCurve=&MakeVector2D(X=1, Y=1)
    ))
   &ScaleColor(Scale RGB=&VectorFromFloat(Value=50), ScaleA=true, ScaleRGB=true)
  ]
  rendererClasses [&NiagaraSpriteRendererProperties]
  renderers [{rendererClass=&NiagaraSpriteRendererProperties}]
 }
 Jet_BurstSparksHigh001 {
  emitterUpdateScript [
   &EmitterState()
   &SpawnRate(
    SpawnRate=&FloatFromCurve(
     FloatCurve=DataInterface() {propertyValues {Curve {keys [
         {interpMode=RCIM_Cubic, tangentMode=RCTM_User, time=0.1, value=1}
         {arriveTangent=-2.5652, interpMode=RCIM_Cubic, leaveTangent=-2.5652, tangentMode=RCTM_User, time=1, value=0.1}
        ]}}}
     CurveIndex=$Emitter.NormalizedLoopAge
     Scale Curve=150
    )
    Use Spawn Probability=false
   )
  ]
  particleSpawnScript [
   &FN_SetSpawnGroup(Number of Spawn Groups=2)
   &InitializeParticle(
    Color=&RandomRangeLinearColor(Minimum=<rgba 0.5071 0 0.6406 1>, Maximum=<rgba 0.0897 0.0722 1 1>)
    Lifetime=&Lerp_Float(
     A=0.15
     B=0.9
     Alpha=&FloatFromCurve(
      FloatCurve=DataInterface() {propertyValues {Curve {keys [
          {interpMode=RCIM_Cubic, tangentMode=RCTM_User, time=0, value=0}
          {arriveTangent=0.9384, interpMode=RCIM_Cubic, leaveTangent=0.9384, tangentMode=RCTM_User, time=0.85, value=0.25}
          {arriveTangent=8.8331, interpMode=RCIM_Cubic, leaveTangent=8.8331, tangentMode=RCTM_User, time=1, value=1}
         ]}}}
      CurveIndex=&RandomRangeFloat(Minimum=0, Maximum=1)
     )
    )
    Module.Write Color=true
    Module.Write Lifetime=true
    Module.Write Mass=false
    Module.Write Material Random=true
    Module.Write Position=false
    Module.Write Scale=false
    Module.Write SpriteRotation=true
    Module.Write SpriteSize=true
    Position=$Engine.Owner.Position
    Sprite Size=&Multiply_Vector2D_ByFloat(
     Float=&Lerp_Float(
      A=1
      B=3
      Alpha=&FloatFromCurve(
       FloatCurve=DataInterface() {propertyValues {Curve {keys [
           {interpMode=RCIM_Cubic, tangentMode=RCTM_User, time=0, value=0}
           {arriveTangent=0.9384, interpMode=RCIM_Cubic, leaveTangent=0.9384, tangentMode=RCTM_User, time=0.85, value=0.25}
           {arriveTangent=8.8331, interpMode=RCIM_Cubic, leaveTangent=8.8331, tangentMode=RCTM_User, time=1, value=1}
          ]}}}
       CurveIndex=&RandomRangeFloat(Minimum=0, Maximum=1)
      )
     )
     Vector2D=&UniformRangedVector2D(Minimum=<5 7>, Maximum=<4 8>)
    )
   )
   &AddVelocityInCone(
    Cone Angle=5
    Module.Use Velocity Falloff On Cone Axis=true
    Velocity Distribution Along Cone Axis=0.6
    Velocity Falloff Away From Cone Axis=0.333
    Velocity Strength=&Lerp_Float(
     A=900
     B=1800
     Alpha=&FloatFromCurve(
      FloatCurve=DataInterface() {propertyValues {Curve {keys [
          {interpMode=RCIM_Cubic, tangentMode=RCTM_User, time=0, value=0}
          {arriveTangent=0.9384, interpMode=RCIM_Cubic, leaveTangent=0.9384, tangentMode=RCTM_User, time=0.8, value=0.35}
          {arriveTangent=4.6474, interpMode=RCIM_Cubic, leaveTangent=4.6474, tangentMode=RCTM_User, time=1, value=1}
         ]}}}
      CurveIndex=&RandomRangeFloat(Minimum=0, Maximum=1)
     )
    )
   )
   &SkeletalMeshLocation(
    Bone / Socket Index=&SelectIntFromArray(Direct Array Index=$Particles.SpawnGroup, Int Selection Array=DataInterface() {propertyValues {IntData [12,13]}})
    Interpolate Mesh Sampling=false
    MaskSpawnGroup=false
    Mesh Sampling Type=::NewEnumerator5
    Orientation Sampling=::NewEnumerator1
    Sampled Position Offset=&SelectVectorFromArray(Direct Array Index=$Particles.SpawnGroup, Vector Selection Array=DataInterface() {propertyValues {FloatData [<-5 0 0>, <-5 0 0>]}})
    Skeletal Mesh=$User.Vehicle_Selection_02
    Socket Sampling Mode=::NewEnumerator5
   )
   &SphereLocation(MaskSpawnGroup=false, Module.Override Randomness=false, Sphere Radius=10, Surface Only=false)
  ]
  particleUpdateScript [
   &ParticleState()
   &CurlNoiseForce(
    Noise Frequency=100
    Noise Strength=&Lerp_Float(
     A=1500
     B=6500
     Alpha=&FloatFromCurve(
      FloatCurve=DataInterface() {propertyValues {Curve {keys [
          {interpMode=RCIM_Cubic, tangentMode=RCTM_User, time=0, value=0}
          {arriveTangent=0.9384, interpMode=RCIM_Cubic, leaveTangent=0.9384, tangentMode=RCTM_User, time=0.8, value=0.35}
          {arriveTangent=4.6474, interpMode=RCIM_Cubic, leaveTangent=4.6474, tangentMode=RCTM_User, time=1, value=1}
         ]}}}
      CurveIndex=&RandomRangeFloat(Minimum=0, Maximum=1)
     )
    )
    Pan Noise=false
    Random Seed=1
   )
   &Drag(Drag=5, Use Linear Drag=true, Use Rotational Drag=false)
   &SolveForcesAndVelocity(Clamp Velocity=false, Limit Acceleration=false)
   &ScaleSpriteSize(Scale Factor=&Vector2DFromCurve(
     Vector2Curve=DataInterface() {propertyValues {
       XCurve {keys [<at -0.0017 0.4556 RCIM_Cubic>, {arriveTangent=-0.9208, interpMode=RCIM_Cubic, tangentMode=RCTM_Break, time=0.1932, value=0.1104}]}
       YCurve {keys [<at 0 1 RCIM_Cubic>,<at 0.6807 0.0224 RCIM_Cubic>]}
      }}
     ScaleCurve=&MakeVector2D(X=1, Y=1)
    ))
   &ScaleColor(Scale RGB=&VectorFromFloat(Value=50), ScaleA=true, ScaleRGB=true)
  ]
  rendererClasses [&NiagaraSpriteRendererProperties]
  renderers [{rendererClass=&NiagaraSpriteRendererProperties}]
 }
 NeonTireRings {
  emitterUpdateScript [
   &EmitterState()
   &SpawnRate(
    Spawn Probability=&MakeCustomFloatFromBool(Bool=$User.IsWaterDeep, True Float=0, False Float=1)
    SpawnRate=&FloatFromCurve(
     FloatCurve=DataInterface() {propertyValues {Curve {keys [
         {interpMode=RCIM_Cubic, tangentMode=RCTM_User, time=0.3, value=1}
         {arriveTangent=-2.5652, interpMode=RCIM_Cubic, leaveTangent=-2.5652, tangentMode=RCTM_User, time=1, value=0.25}
        ]}}}
     CurveIndex=$Emitter.NormalizedLoopAge
     Scale Curve=&Lerp_Float(A=25, B=35, Alpha=&MakeCustomFloatFromBool(Bool=$User.Mod_OffroadTires, True Float=1, False Float=0))
    )
    Use Spawn Probability=true
   ) {enabled=false}
   &SpawnRate(
    Spawn Probability=&MakeCustomFloatFromBool(Bool=&Boolean_OrOperation(Bool1=$User.IsWaterDeep, Bool2=$User.TireIntactRL), True Float=0, False Float=1)
    SpawnRate=&FloatFromCurve(
     FloatCurve=DataInterface() {propertyValues {Curve {keys [
         {interpMode=RCIM_Cubic, tangentMode=RCTM_User, time=0.3, value=1}
         {arriveTangent=-2.5652, interpMode=RCIM_Cubic, leaveTangent=-2.5652, tangentMode=RCTM_User, time=1, value=0.25}
        ]}}}
     CurveIndex=$Emitter.NormalizedLoopAge
     Scale Curve=&Lerp_Float(A=25, B=35, Alpha=&MakeCustomFloatFromBool(Bool=$User.Mod_OffroadTires, True Float=1, False Float=0))
    )
    Use Spawn Probability=true
   ) {module=SpawnRate001}
   &SpawnRate(
    Spawn Probability=&MakeCustomFloatFromBool(Bool=&Boolean_OrOperation(Bool1=$User.IsWaterDeep, Bool2=$User.TireIntactRR), True Float=0, False Float=1)
    SpawnGroup=1
    SpawnRate=&FloatFromCurve(
     FloatCurve=DataInterface() {propertyValues {Curve {keys [
         {interpMode=RCIM_Cubic, tangentMode=RCTM_User, time=0.3, value=1}
         {arriveTangent=-2.5652, interpMode=RCIM_Cubic, leaveTangent=-2.5652, tangentMode=RCTM_User, time=1, value=0.25}
        ]}}}
     CurveIndex=$Emitter.NormalizedLoopAge
     Scale Curve=&Lerp_Float(A=25, B=35, Alpha=&MakeCustomFloatFromBool(Bool=$User.Mod_OffroadTires, True Float=1, False Float=0))
    )
    Use Spawn Probability=true
   ) {module=SpawnRate002}
  ]
  particleSpawnScript [
   &PartitionParticles() {enabled=false}
   /Niagara/Modules/Spawn/Initialization/V2/InitializeParticle(
    Color=$System.VehicleColor
    Lifetime Max=0.5
    Lifetime Min=0.25
    Lifetime Mode=::NewEnumerator1
    Mesh Scale Mode=::NewEnumerator3
    Mesh Uniform Scale=&Lerp_Float(
     A=0.3
     B=0.32
     Alpha=&MakeCustomFloatFromBool(Bool=$User.Mod_OffroadTires, True Float=1, False Float=0)
    )
    UsePositionOffset=false
    Write Material Random=true
   )
   &SkeletalMeshLocation(
    Bone / Socket Index=&SpawnGroup(Max Spawn Group=1)
    Interpolate Mesh Sampling=false
    MaskSpawnGroup=false
    Mesh Sampling Type=::NewEnumerator5
    Sampled Position Offset=&Multiply_Vector(
     A=&SelectVectorFromArray(
      Direct Array Index=&SpawnGroup()
      Vector Selection Array=DataInterface() {propertyValues {FloatData [<0 28 0>, <0 -28 0>]}}
     )
     B=&Lerp_Vector(
      A=<0 1 1>
      B=<0 1.1 1>
      Alpha=&MakeCustomFloatFromBool(Bool=$User.Mod_OffroadTires, True Float=1, False Float=0)
     )
    )
    Skeletal Mesh=$User.Vehicle_Selection_02
    Socket Sampling Mode=::NewEnumerator5
   )
   &ShapeLocation(
    Box Size=&Lerp_Vector(
     A=<0 5 0>
     B=<0 10 0>
     Alpha=&MakeCustomFloatFromBool(Bool=$User.Mod_OffroadTires, True Float=1, False Float=0)
    )
    MaskSpawnGroup=false
    Override Seed=false
    Shape Primitive=::NewEnumerator2
    Surface Only=false
    UseOwnerScale=true
   )
   &InitialMeshOrientation(
    Mesh Axis To Orient=<0 1 0>
    Mesh Orientation Mode=::NewEnumerator2
    Orientation Vector=<0 0 1>
    Rotation=&Multiply_Vector(
     A=&SelectVectorFromArray(
      Direct Array Index=&SpawnGroup()
      Vector Selection Array=DataInterface() {propertyValues {FloatData [<0 0 0>, <0 0 0>]}}
     )
     B=&Lerp_Vector(
      A=<1 0 0>
      B=<0 0 0>
      Alpha=&MakeCustomFloatFromBool(Bool=$User.Mod_OffroadTires, True Float=1, False Float=0)
     )
    )
    Use Rotation Vector=true
   )
   &AddVelocity(
    Velocity=&SelectVectorFromArray(
     Direct Array Index=&SpawnGroup()
     Vector Selection Array=DataInterface() {propertyValues {FloatData [<0 40 0>, <0 -40 0>]}}
    )
    Velocity Speed Scale=&Lerp_Float(
     A=1.25
     B=0.75
     Alpha=&FloatFromCurve(
      FloatCurve=DataInterface() {propertyValues {Curve {keys [
          {interpMode=RCIM_Cubic, tangentMode=RCTM_User, time=0.1, value=0}
          {arriveTangent=1.5933, interpMode=RCIM_Cubic, leaveTangent=1.5933, tangentMode=RCTM_User, time=0.5, value=0.2}
          {interpMode=RCIM_Cubic, tangentMode=RCTM_User, time=0.9, value=1}
         ]}}}
      CurveIndex=$Emitter.NormalizedLoopAge
     )
    )
   )
  ]
  particleUpdateScript [
   &ParticleState()
   &Drag(Use Linear Drag=true, Use Rotational Drag=false)
   &ScaleMeshSize(Scale Factor=&VectorFromFloat(Value=&FloatFromCurve(FloatCurve=DataInterface() {propertyValues {Curve {keys [
          {interpMode=RCIM_Cubic, tangentMode=RCTM_User, time=0.3, value=1}
          {arriveTangent=-0.5528, interpMode=RCIM_Cubic, leaveTangent=-0.5528, tangentMode=RCTM_User, time=1, value=0.75}
         ]}}})))
   &DynamicMaterialParameters(
    Index 0 Param 1=10
    Index 0 Param 3=1
    Module.Param0WriteEnabled=true
    Module.Param1WriteEnabled=true
    Module.Param2WriteEnabled=true
    Module.Param3WriteEnabled=true
   )
   &ScaleColor(
    Scale Alpha=&FloatFromCurve(FloatCurve=DataInterface() {propertyValues {Curve {keys [{interpMode=RCIM_Cubic, tangentMode=RCTM_User, time=0, value=1}, {interpMode=RCIM_Cubic, tangentMode=RCTM_User, time=1, value=0}]}}})
    Scale RGB=&VectorFromFloat(Value=&Lerp_Float(
      A=25
      B=10
      Alpha=&FloatFromCurve(
       FloatCurve=DataInterface() {propertyValues {Curve {keys [
           {interpMode=RCIM_Cubic, tangentMode=RCTM_User, time=0.1, value=0}
           {arriveTangent=1.5933, interpMode=RCIM_Cubic, leaveTangent=1.5933, tangentMode=RCTM_User, time=0.5, value=0.2}
           {interpMode=RCIM_Cubic, tangentMode=RCTM_User, time=0.9, value=1}
          ]}}}
       CurveIndex=$Emitter.NormalizedLoopAge
      )
     ))
    ScaleA=true
    ScaleRGB=true
   )
   &SolveForcesAndVelocity(Clamp Velocity=false, Limit Acceleration=false)
  ]
  rendererClasses [&NiagaraMeshRendererProperties]
  renderers [{rendererClass=&NiagaraMeshRendererProperties, rendererIndex=0}]
 }
}
summary {
 systemName=NS_Boost_SuperSedan
 emitters [
  {emitterName=ExhaustFlame, rendererClasses [&NiagaraMeshRendererProperties]}
  {emitterName=ExhaustLight, rendererClasses [&NiagaraLightRendererProperties, &NiagaraSpriteRendererProperties]}
  {emitterName=Jet_BurstSparksHigh, simTarget=GPUComputeSim, rendererClasses [&NiagaraSpriteRendererProperties]}
  {emitterName=HotSphere, rendererClasses [&NiagaraMeshRendererProperties]}
  {emitterName=Jet_BurstSparksHigh001, rendererClasses [&NiagaraSpriteRendererProperties]}
  {emitterName=NeonTireRings, rendererClasses [&NiagaraMeshRendererProperties]}
  {
   bEnabled=false
   emitterName=ExhaustLight001
   rendererClasses [&NiagaraLightRendererProperties, &NiagaraSpriteRendererProperties]
  }
 ]
 userVariables [
  User.BoostFX {type=NiagaraBool}
  User.Downshifting {type=NiagaraBool}
  User.Dust_Particles_Per_Wheel {type=NiagaraInt32, defaultValue=18}
  User.Dust_SpawnCurve {
   type=NiagaraDataInterfaceCurve
   description=""
   defaultValue {
    struct=NiagaraExt_VariableValue_DataInterface
    value {
     dataInterface=&self:NiagaraDataInterfaceCurve_0
     dataInterfaceClass=&NiagaraDataInterfaceCurve
    }
   }
  }
  User.Dust_Wheel_RL {type=NiagaraBool}
  User.Dust_Wheel_RR {type=NiagaraBool}
  User.DustSpawnPerUnit {type=NiagaraFloat, defaultValue=150, description=""}
  User.DustTireBlurColor {type=LinearColor, description="", defaultValue {struct=LinearColor, value=<rgba 1 1 1 1>}}
  User.ExhaustColor {type=LinearColor, description="", defaultValue {struct=LinearColor, value=<rgba 0.6062 0.754 1 0.025>}}
  User.ExhaustScale {type=NiagaraFloat, defaultValue=1, description=""}
  User.HasQuadExhaust {type=NiagaraBool}
  User.IsBigRig {type=NiagaraBool}
  User.IsDoingStoppie {type=NiagaraBool}
  User.IsDrifting {type=NiagaraBool}
  User.IsPeelOutOnRoad {type=NiagaraBool}
  User.IsPeelOutOnRoadRewind {type=NiagaraBool}
  User.IsShifting {type=NiagaraBool}
  User.IsWaterDeep {type=NiagaraBool}
  User.JetFlameHight {type=NiagaraFloat, defaultValue=0, description=""}
  User.JetLocationOffsetZ {type=NiagaraFloat, defaultValue=0, description=""}
  User.JetSparksCounts {type=NiagaraFloat, defaultValue=0, description=""}
  User.Mod_OffroadTires {type=NiagaraBool}
  User.New Skeletal Mesh {
   type=NiagaraDataInterfaceSkeletalMesh
   defaultValue {
    struct=NiagaraExt_VariableValue_DataInterface
    value {dataInterface=&self:NiagaraDataInterfaceSkeletalMesh_3}
   }
  }
  User.ParticleCountPerWheel {type=NiagaraInt32, defaultValue=150}
  User.PeelOut_Wheel_Rear {type=NiagaraBool}
  User.PeelOut_Wheel_RL {type=NiagaraBool, defaultValue=true}
  User.PeelOut_Wheel_RR {type=NiagaraBool, defaultValue=true}
  User.PeelOutColorArray {
   type=NiagaraDataInterfaceArrayColor
   description=""
   defaultValue {
    struct=NiagaraExt_VariableValue_DataInterface
    value {
     dataInterface=&self:NiagaraDataInterfaceArrayColor_0
     dataInterfaceClass=&NiagaraDataInterfaceArrayColor
    }
   }
  }
  User.PeelOutColorIndex {type=NiagaraInt32}
  User.QuadExhaustHorizontalOffset {type=NiagaraFloat, defaultValue=0, description=""}
  User.RibbonIDOffset {type=NiagaraInt32}
  User.SedanColor {type=NiagaraInt32}
  User.SkeletalMesh {
   type=NiagaraDataInterfaceSkeletalMesh
   defaultValue {
    struct=NiagaraExt_VariableValue_DataInterface
    value {dataInterface=&self:NiagaraDataInterfaceSkeletalMesh_2}
   }
  }
  User.SkeletalMesh_Vehicle_Neon {
   type=NiagaraDataInterfaceSkeletalMesh
   defaultValue {
    struct=NiagaraExt_VariableValue_DataInterface
    value {dataInterface=&self:NiagaraDataInterfaceSkeletalMesh_0}
   }
  }
  User.Skid_Wheels {type=NiagaraBool}
  User.Slide_Direction {type=NiagaraFloat, defaultValue=0, description=""}
  User.Sparks_Wheel_FL {type=NiagaraBool}
  User.Sparks_Wheel_FR {type=NiagaraBool}
  User.Sparks_Wheel_ML {type=NiagaraBool}
  User.Sparks_Wheel_MR {type=NiagaraBool}
  User.Sparks_Wheel_RL {type=NiagaraBool}
  User.Sparks_Wheel_RR {type=NiagaraBool}
  User.SparkSPU_MaxSpeed {type=NiagaraFloat, defaultValue=3, description=""}
  User.SparkSPU_MinSpeed {type=NiagaraFloat, defaultValue=4, description=""}
  User.Tire_Scale_Back {type=Vector3f, description="", defaultValue {struct=Vector3f, value=<0.39 0.4325 0.39>}}
  User.TireIntactFL {type=NiagaraBool, defaultValue=true}
  User.TireIntactFR {type=NiagaraBool, defaultValue=true}
  User.TireIntactRL {type=NiagaraBool}
  User.TireIntactRR {type=NiagaraBool}
  User.Vehicle_Selection_02 {
   type=NiagaraDataInterfaceSkeletalMesh
   defaultValue {
    struct=NiagaraExt_VariableValue_DataInterface
    value {dataInterface=&self:NiagaraDataInterfaceSkeletalMesh_1}
   }
  }
  User.Vehicle_Selection_Exhaust {
   type=NiagaraDataInterfaceSkeletalMesh
   defaultValue {
    struct=NiagaraExt_VariableValue_DataInterface
    value {dataInterface=&self:NiagaraDataInterfaceSkeletalMesh_4}
   }
  }
  User.Vehicle_Skeleton {
   type=NiagaraDataInterfaceSkeletalMesh
   defaultValue {
    struct=NiagaraExt_VariableValue_DataInterface
    value {dataInterface=&self:NiagaraDataInterfaceSkeletalMesh_5}
   }
  }
  User.Vehicle_Speed {type=NiagaraFloat, defaultValue=0, description=""}
  User.VehicleHasDriver {type=NiagaraBool}
  User.VehicleHasFuel {type=NiagaraBool}
  User.Wheel_Scale_Back {type=Vector3f, description="", defaultValue {struct=Vector3f, value=<0.3 0.3 0.3>}}
  User.Wheel_Scale_Front {type=Vector3f, description="", defaultValue {struct=Vector3f, value=<0.3 0.3 0.3>}}
  User.Wheel_Width {type=NiagaraFloat, defaultValue=16, description=""}
  User.Wheel_Width_Back {type=NiagaraFloat, defaultValue=18, description=""}
  User.WheelsTouchingArray {
   type=NiagaraDataInterfaceArrayFloat
   defaultValue {
    struct=NiagaraExt_VariableValue_DataInterface
    value {
     dataInterface=&self:NiagaraDataInterfaceArrayFloat_0
     dataInterfaceClass=&NiagaraDataInterfaceArrayFloat
    }
   }
  }
 ]
}
