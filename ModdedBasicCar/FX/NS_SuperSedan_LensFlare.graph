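# Niagara lens flare for the SuperSedan: two GPU sprites at mirrored side offsets, visible only in evening and night, tinted by a system-level vehicle color.
# Four sprite renderers draw each particle, with spawn-time offsets and sizes set for three streaks.
# A scratch module samples occlusion in rings, with visibility scaling alpha.
# A camera-vector component drives sprite size, brightness, alpha and rotation through curves, making the flare respond to viewing angle.
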NiagaraSystem /ModdedBasicCar/FX/NS_SuperSedan_LensFlare
dependencies {
 usedDataInterfaces [
  &NiagaraDataInterfaceArrayFloat3
  /Script/Niagara.NiagaraDataInterfaceOcclusion
  &NiagaraDataInterfaceCamera
  &NiagaraDataInterfaceCurve
  &NiagaraDataInterfaceArrayColor
 ]
 usedDynamicInputs [&SelectLinearColorFromArray]
 usedModules [
  &SystemState
  &self:SystemSpawnScript.SystemScriptSource.SystemScriptGraph.NiagaraNodeAssignment_0.SetVariables_#1
 ]
}
emitters {LensFlare {
  simTarget=GPUComputeSim
  emitterUpdateScript [
   &EmitterState(Life Cycle Mode=::NewEnumerator1, Loop Duration=1, UseLoopDelay=false)
   &SpawnBurst_Instantaneous(Spawn Count=2, Use Spawn Probability=false, UseLoopCountLimit=false)
  ]
  particleSpawnScript [
   &PartitionParticles()
   &InitializeParticle(
    Color=$System.VehicleColor
    Position Offset=&SelectVectorFromArray(
     Direct Array Index=$Particles.PartitionParticles.Partition
     Vector Selection Array=DataInterface() {propertyValues {FloatData [<-12 60 0>, <-12 -60 0>]}}
    )
    Sprite Rotation Mode=::NewEnumerator2
    Sprite Size Mode=::NewEnumerator3
    Uniform Sprite Size=80
    UsePositionOffset=true
    Write Material Random=true
   )
   /SetVariables(
    Particles.Streak_01_Offset:NiagaraPosition=&AddVectorToPosition(Position=$Particles.Position, Vector=<0 0 5>)
    Particles.Streak_01_Size:Vector2f=<300 10>
    Particles.Streak_02_Offset:NiagaraPosition=&AddVectorToPosition(Position=$Particles.Position, Vector=<0 0 1.5>)
    Particles.Streak_02_Size:Vector2f=<200 10>
    Particles.Streak_03_Offset:NiagaraPosition=&AddVectorToPosition(Position=$Particles.Position, Vector=<0 0 -2>)
    Particles.Streak_03_size:Vector2f=<150 10>
   ) {isSetParameters=true}
  ]
  particleUpdateScript [
   &ParticleState(Kill Particles When Lifetime Has Elapsed=false, Let Infinitely Lived Particles Die When Emitter Deactivates=true)
   &self:ScratchModule(NumberOfRings=5, OcclusionSize=20, SamplesPerRing=5) {module=ScratchModule}
   /SetVariables(Particles.CameraScale:NiagaraFloat=&Multiply_Float(
     A=1
     B=&MakeFloatFromVector(
      Vector=&ReturnCameraVector(Camera Vectors=::NewEnumerator4, Coordinate Space=::Simulation, Origin Position=$Particles.Position)
      Channel=::NewEnumerator0
     )
    )) {isSetParameters=true}
   &ScaleSpriteSize(
    Uniform Curve Index=$Particles.CameraScale
    Uniform Curve Sprite Scale=DataInterface() {propertyValues {Curve {keys [
        {interpMode=RCIM_Cubic, tangentMode=RCTM_User, time=0, value=0}
        {arriveTangent=0.9156, interpMode=RCIM_Cubic, leaveTangent=0.9156, tangentMode=RCTM_User, time=0.3, value=1.275}
        {tangentMode=RCTM_User, time=1, value=1.5}
       ]}}}
   )
   &ScaleSpriteSize(
    Uniform Curve Index=$Particles.CameraScale
    Uniform Curve Sprite Scale=DataInterface() {propertyValues {Curve {keys [
        {tangentMode=RCTM_User, time=0.3, value=0}
        {arriveTangent=2.1429, interpMode=RCIM_Cubic, leaveTangent=2.1429, time=0.7, value=0.55}
        {tangentMode=RCTM_User, time=1, value=1.5}
       ]}}}
   ) {enabled=false, module=ScaleSpriteSize001}
   &ScaleColor(Scale Alpha=$Particles.Visibility, Scale RGB=&VectorFromFloat(Value=3), ScaleA=true, ScaleRGB=true) {module=ScaleColor002}
   &ScaleColor(
    Scale Alpha=&FloatFromCurve(
     FloatCurve=DataInterface() {propertyValues {Curve {keys [
         {interpMode=RCIM_Cubic, tangentMode=RCTM_User, time=0, value=0}
         {arriveTangent=0.4636, interpMode=RCIM_Cubic, leaveTangent=0.4636, tangentMode=RCTM_User, time=0.2, value=0.8}
         {arriveTangentWeight=1.2316, interpMode=RCIM_Cubic, tangentMode=RCTM_User, time=1, value=1}
        ]}}}
     CurveIndex=$Particles.CameraScale
    )
    Scale RGB=&Multiply_VectorByFloat(Vector=<1 1 1>, Float=$Particles.CameraScale)
    ScaleA=true
    ScaleRGB=true
   )
   &ScaleColor(
    Scale Alpha=&FloatFromCurve(
     FloatCurve=DataInterface() {propertyValues {Curve {keys [
         {tangentMode=RCTM_User, time=0, value=0}
         {arriveTangent=0.4636, interpMode=RCIM_Cubic, leaveTangent=0.4636, tangentMode=RCTM_User, time=0.3, value=0}
         {arriveTangent=4.5091, interpMode=RCIM_Cubic, leaveTangent=4.5091, tangentMode=RCTM_User, time=1, value=1}
        ]}}}
     CurveIndex=$Particles.CameraScale
    )
    Scale RGB=&Multiply_VectorByFloat(Vector=<1 1 1>, Float=$Particles.CameraScale)
    ScaleA=true
    ScaleRGB=true
   ) {enabled=false, module=ScaleColor003}
   /SetVariables(Particles.SpriteRotation:NiagaraFloat=&FloatFromCurve(
     FloatCurve=DataInterface() {propertyValues {Curve {keys [<at 0 -1>,<at 0.5 1>,<at 1 -1>]}}}
     CurveIndex=$Particles.CameraScale
     Scale Curve=30
    )) {isSetParameters=true}
   /SetVariables(
    Particles.Streak_01_Size:Vector2f=&Multiply_Vector2D_ByFloat(Float=1, Vector2D=$Particles.Initial.Streak_01_Size)
    Particles.Streak_02_Size:Vector2f=&Multiply_Vector2D_ByFloat(Float=1, Vector2D=$Particles.Initial.Streak_02_Size)
   ) {enabled=false, isSetParameters=true}
   /SetVariables(
    Particles.Streak_01_Size:Vector2f=&Multiply_Vector2D_ByFloat(
     Float=&FloatFromCurve(
      FloatCurve=DataInterface() {propertyValues {Curve {keys [
          {interpMode=RCIM_Cubic, tangentMode=RCTM_User, time=0, value=0}
          {arriveTangent=0.9156, interpMode=RCIM_Cubic, leaveTangent=0.9156, tangentMode=RCTM_User, time=0.3, value=1.275}
          {tangentMode=RCTM_User, time=1, value=1.5}
         ]}}}
      CurveIndex=$Particles.CameraScale
     )
     Vector2D=$Particles.Initial.Streak_01_Size
    )
    Particles.Streak_02_Size:Vector2f=&Multiply_Vector2D_ByFloat(
     Float=&FloatFromCurve(
      FloatCurve=DataInterface() {propertyValues {Curve {keys [
          {interpMode=RCIM_Cubic, tangentMode=RCTM_User, time=0, value=0}
          {arriveTangent=0.9156, interpMode=RCIM_Cubic, leaveTangent=0.9156, tangentMode=RCTM_User, time=0.3, value=1.275}
          {tangentMode=RCTM_User, time=1, value=1.5}
         ]}}}
      CurveIndex=$Particles.CameraScale
     )
     Vector2D=$Particles.Initial.Streak_02_Size
    )
   ) {enabled=false, isSetParameters=true}
   &ScaleColor(
    Scale Alpha=&FN_Select_Float_from_Time_Of_Day(Morning=0, Day=0, Evening=1, Night=1)
    ScaleA=true
    ScaleRGB=false
   ) {module=ScaleColor001}
  ]
  rendererClasses [&NiagaraSpriteRendererProperties]
  renderers [
   {rendererClass=&NiagaraSpriteRendererProperties}
   {rendererClass=&NiagaraSpriteRendererProperties, rendererIndex=1}
   {rendererClass=&NiagaraSpriteRendererProperties, rendererIndex=2}
   {rendererClass=&NiagaraSpriteRendererProperties, rendererIndex=3}
  ]
 }}
summary {
 systemName=NS_SuperSedan_LensFlare
 emitters [{emitterName=LensFlare, simTarget=GPUComputeSim, rendererClasses [&NiagaraSpriteRendererProperties]}]
 userVariables [User.SedanColor {type=NiagaraInt32}]
}
